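; Athena player campfire floor trap that heals player pawns and vehicles in a box around it.
; On placement the server spawns and attaches a campfire gameplay actor, silently killing it when the trap dies.
; HealTicks applies the heal GameplayEffect up to a data-driven tick limit, sending squadmate/teammate quest stats, then clears IsActive.
; OnRep_IsActive drives cosmetics: fire, smoke, audio, and timelines that char the logs and fade the doused effect.
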(event EventOnDeathPlayEffects
  (params (Damage :type Float%20%28single-precision%29) (DamageTags :type Gameplay%20Tag%20Container%20Structure%20%28by%20ref%29) (Momentum :type Vector :default "0, 0, 0") (HitInfo :type Hit%20Result%20Structure%20%28by%20ref%29) (InstigatedBy :type Fort%20Pawn%20Object%20Reference) (DamageCauser :type Actor%20Object%20Reference) (EffectContext :type Gameplay%20Effect%20Context%20Handle%20Structure))
  ;[1 1 1] Stop audio on client destroy
  (Audio|Components|Audio|FadeOut (Variables|Default|GetAudio) 0.2 0.0 "Linear" :!class /Script/Engine.AudioComponent))

(event Custom|DousedTime
  (bind _p_athena_campfire_doused (Variables|Default|GetP_Athena_Campfire_Doused))
  ;[1 1 1] Doused Time - delay and turn off doused FX
  (Components|Timeline|SetPlayRate (Variables|Trap_Floor_Player_Campfire|GetFadeDoused) (/ 1.0 (ScalableFloat|GetValueatLevel (Variables|Rows|GetRow_DousedDuration))))
  (bind (_direction _fade) (Timeline|FadeDoused PlayFromStart)
    (:Update
      (Effects|Components|ParticleSystem|SetFloatParameter _p_athena_campfire_doused "Alpha" _fade))
    (:Finished
      (Components|Activation|Deactivate _p_athena_campfire_doused))))

(event EventOnPlaced
  ;[1 1 1] Initial Setup
  (Variables|Mesh|SetbNoCollision true) ; For when OnCollisionSettingsNeedUpdate() runs
  (Collision|SetCollisionEnabled (Variables|Debug|GetStaticMeshComponent) "NoCollision") ; Manually Set the collision as it is turned on when placed
  (Variables|Default|SetMaxTicks (Math|Float|Truncate (ScalableFloat|GetValueatLevel (Variables|Rows|GetRow_MaxHeals))))
  (Audio|PlaySoundatLocation (Variables|Audio|GetTrap_Placed_Sound) (Transformation|GetWorldLocation (Variables|Default|GetPlacementSoundLocation)))
  (Utilities|TrapPlacementDustFX :Attach%20to%20Component (Variables|Debug|GetStaticMeshComponent) :Socket "PlacedFX" :Rotation "0, 0, 0" :Burst%20Rate 8 :Size%20%28Volume%29 "100.000000,100.000000,15.000000" :Velocity%20Scale "1,1,1" :Size%20Scale%20%28Particles%29 1
    (:Out
      (Networking|SwitchHasAuthority
        (:Authority
          (bind _returnvalue (Building|SpawnBuildingGameplayActor (Variables|Default|GetCampFire%20BGA%20Trap%20Class) (Math|Conversions|ToTransform(Vector) (Transformation|GetWorldLocation (Variables|Default|GetCampfirePit))) self))
          (bind _output_get (Variables|Default|SetCampFire_BGA _returnvalue))
          (Transformation|AttachActorToComponent _output_get (Variables|Default|GetCampfirePit) "None" "SnapToTarget" "KeepRelative" "KeepRelative" true))))))

(event EventOnDeathServer
  (params (Damage :type Float%20%28single-precision%29) (DamageTags :type Gameplay%20Tag%20Container%20Structure%20%28by%20ref%29) (Momentum :type Vector :default "0, 0, 0") (HitInfo :type Hit%20Result%20Structure%20%28by%20ref%29) (InstigatedBy :type Controller%20Object%20Reference) (DamageCauser :type Actor%20Object%20Reference) (EffectContext :type Gameplay%20Effect%20Context%20Handle%20Structure))
  (Gameplay|SilentDie (Variables|Default|GetCampFire_BGA) false))

(event EventOnFinishedBuilding
  ;[1 1 1] Finished building
  (Audio|PlaySoundatLocation (Variables|Audio|GetTrap_Active_Sound) (Transformation|GetActorLocation))
  (Fortnite|ShowSoundIndicatorLocalforTeam :Location (Transformation|GetActorLocation) :MaxAudibleDistance (Variables|/Script/Engine.SoundBase|GetMaxDistance (Variables|Audio|GetTrap_Placed_Sound)) :IndicatorType "Gunshot" :Team (Team|GetActorTeam self) :Affiliations (Utilities|Array|MakeArray "Hostile" "Neutral") :Tint "(R=0.000000,G=0.000000,B=0.000000,A=0.000000)" :OverrideIcon "/Game/UI/Foundation/Textures/Icons/Navigation/t_icon_nav_mobile_audioviz_building_32.t_icon_nav_mobile_audioviz_building_32" :AttenuationMaxRangeMultiplier 1.0 :AttenuationIntensityMultiplier 1.0))

(event GameplayCue|GameplayCue.Abilities.Activation.Traps.ActivateTrap
  (params (EventType :type EGameplayCueEvent%20Enum :default "OnActive") (Parameters :type Gameplay%20Cue%20Parameters%20Structure%20%28by%20ref%29))
  ;[1 1 1] On Fire
  (switch Utilities|FlowControl|Switch|SwitchonEGameplayCueEvent EventType
    (:Executed
      (Audio|PlaySoundatLocation (Variables|Audio|GetTrap_Fire_Sound) (Transformation|GetActorLocation) "0, 0, 0" 1.5 0.75))))

(event GameplayCue|GameplayCue.Abilities.Activation.Traps.ReloadBegin
  (params (EventType :type EGameplayCueEvent%20Enum :default "OnActive") (Parameters :type Gameplay%20Cue%20Parameters%20Structure%20%28by%20ref%29))
  ;[1 1 1] Reload Begin
  (switch Utilities|FlowControl|Switch|SwitchonEGameplayCueEvent EventType
    (:Executed
      (Audio|PlaySoundatLocation (Variables|Audio|GetTrap_Reload_Sound) (Transformation|GetActorLocation)))))

(event EventOnReloadEnd
  ;[1 1 1] Reload End
  (Audio|PlaySoundatLocation (Variables|Audio|GetTrap_Active_Sound) (Transformation|GetActorLocation)))

(event EventOnOutOfDurability
  ;[1 1 1] Durability Change
  (Utilities|FlowControl|Delay 1)
  (Audio|PlaySoundatLocation (Variables|Audio|GetTrap_Explode_Sound) (Transformation|GetActorLocation))
  (Effects|SpawnEmitteratLocation "/Game/Weapons/FORT_TRAPS/Spike_Trap/P_Spike_Trap_Destruction.P_Spike_Trap_Destruction" (Transformation|GetWorldLocation (Variables|Transformation|GetRootComponent))))

(event Custom|TriggerLaunchEffects
  (params (Pawn :type Fort%20Player%20Pawn%20Athena%20Object%20Reference)))

(event Custom|CharLogs
  ;[1 1 1] Darken Logs and lerp fire alpha
  (Components|Timeline|SetPlayRate (Variables|Trap_Floor_Player_Campfire|GetDarkenLogs) (/ 1.0 (* (ScalableFloat|GetValueatLevel (Variables|Rows|GetRow_TickInterval)) (ScalableFloat|GetValueatLevel (Variables|Rows|GetRow_MaxHeals)))))
  (bind (_direction _lerp) (Timeline|DarkenLogs PlayFromStart)
    (:Update
      (Rendering|Material|SetVectorParameterValueonMaterials (Variables|Default|GetWood) "Base Color Multiplier" (Math|Vector|Lerp(Vector) "1.000000,1.000000,1.000000" "0.100000,0.100000,0.100000" _lerp))
      (Effects|Components|ParticleSystem|SetFloatParameter (Variables|Default|GetP_Athena_Campfire) "Alpha" (Math|Float|Lerp 0.0 1.0 _lerp)))))

(fn CheckIfTargetIsSquadmate
  (params (Target%20Player :type Fort%20Player%20Pawn%20Object%20Reference))
  (bind _returnvalue (Building|GetOwningPlayerState))
  (if (not (Variables|Default|GetHasFiredQuest))
    ;[1 1 1] Save off references.
    (bind _output_get (Variables|Default|SetHitPlayerPawn Target%20Player))
    (bind _output_get_2 (Variables|Default|SetOwning%20Player%20PC (Fortnite|GetFortPlayerControllerfromActor _returnvalue)))
    (if (and (Utilities|IsValid _output_get_2 :!class /Script/Engine.KismetSystemLibrary) (Utilities|IsValid _output_get :!class /Script/Engine.KismetSystemLibrary))
      (Variables|Default|SetOwning%20Player%20Quest%20Manager (Quests|GetQuestManager _output_get_2 "Invalid"))
      ;[1 1 1] Filter owner out of  checks.
      (if (Utilities|NotEqual(Object) (PlayerState|GetCurrentPawn _returnvalue) (Variables|Default|GetHitPlayerPawn))
        (bind _asfort_player_state_athena (Utilities|Casting|CastToFortPlayerStateAthena (Variables|/Script/Engine.Controller|GetPlayerState (Pawn|GetController (Variables|Default|GetHitPlayerPawn)))))
        ;[1 1 1] Set friendliness level bools.
        (LocalVariables|SetIs%20Squad%20Member (Class|FortPlayerStateAthena|IsinSquad _asfort_player_state_athena))
        (LocalVariables|SetIs%20Team%20Mate (Team|OnSameTeam (Variables|Default|GetOwning%20Player%20PC) _asfort_player_state_athena))
        ;[1 1 1] Send squad and/or teammate healed quest stat events.
        (Utilities|FlowControl|Sequence
          (:then_0
            (if (LocalVariables|GetIs%20Squad%20Member)
              (Quest|SendComplexCustomStatEvent (Variables|Default|GetOwning%20Player%20Quest%20Manager) 0 (LocalVariables|GetHeal%20Squadmate%20TC) "(GameplayTags=)" 1)
              (Variables|Default|SetHasFiredQuest true)))
          (:then_1
            (if (LocalVariables|GetIs%20Team%20Mate)
              (Quest|SendComplexCustomStatEvent (Variables|Default|GetOwning%20Player%20Quest%20Manager) 0 (LocalVariables|GetHeal%20Teammate%20TC) "(GameplayTags=)" 1)
              (Variables|Default|SetHasFiredQuest true))))))))

(fn GetCampFireTrapInstigator
  (bind _output_get (LocalVariables|SetOwningPlayerState (Building|GetOwningPlayerState)))
  (Utilities|IsValid _output_get
    (:Is%20Valid
      (return (PlayerState|GetCurrentPawn _output_get)))
    (:Is%20Not%20Valid
      (return 0))))

(fn HealTicks
  (Networking|SwitchHasAuthority
    (:Authority
      (if (< (Variables|Default|GetNumTicks) (Variables|Default|GetMaxTicks))
        (bind _outactors (Collision|BoxOverlapActors (Transformation|GetWorldLocation (Variables|Default|GetProximityTraceOrigin)) (Variables|Default|GetAOE_BoxExtents) (Variables|Default|GetOverlapObjectTypes) :!class /Script/Engine.KismetSystemLibrary)) ; What Pawns overlap the area around the campfire?
        (for _array_element _outactors
          (bind _asfort_player_pawn_athena (Utilities|Casting|CastToFortPlayerPawnAthena _array_element)
            (:then
              (if (FortPawn|IsDead _asfort_player_pawn_athena)
                _
                (elif (Math|Conversions|ToBoolean(Integer) (Math|Float|Truncate (ScalableFloat|GetValueatLevel (Variables|Rows|GetRow_ShouldStack))))
                  ;[1 1 1] Heal Player
                  (CallFunction|CheckifTargetIsSquadmate _asfort_player_pawn_athena) ; Check if friendly for "heal squadmate/teammate with campfire" quest.
                  (GameplayEffects|FortApplyGameplayEffectToTarget :self (Ability|GetFortAbilitySystemComponent (Building|GetOwningPlayerState)) :GameplayEffectClass "/Game/Athena/Items/Traps/Campfire/Blueprint/GE_Athena_Campfire_Heal.GE_Athena_Campfire_Heal_C" :TargetOfEffect (Ability|GetAbilitySystemComponent _asfort_player_pawn_athena) :OptionalSourceObject self :OptionalEffectCauser self :Level 0.0)
                  (else
                    (if (Math|Geometry|IsPointinBoxwithTransform (Transformation|GetWorldLocation (Variables|/Script/Engine.Character|GetCapsuleComponent _asfort_player_pawn_athena)) (Transformation|GetWorldTransform (Variables|Default|GetProximityTraceOrigin)) (Variables|Default|GetAOE_BoxExtents))
                      _)))))
            (:CastFailed
              (Utilities|Casting|CastToFortVehicleInterface _array_element)
              ;[1 1 1] Heal Vehicle
              (if (Math|Conversions|ToBoolean(Integer) (Math|Float|Truncate (ScalableFloat|GetValueatLevel (Variables|Rows|GetRow_ShouldStack))))
                _
                (elif (Math|Geometry|IsPointinBoxwithTransform (Transformation|GetWorldLocation (Variables|/Script/Engine.Actor|GetRootComponent _array_element)) (Transformation|GetWorldTransform (Variables|Default|GetProximityTraceOrigin)) (Variables|Default|GetAOE_BoxExtents))
                  _))
              (Environment|ApplyGlobalEnvironmentGameplayEffecttoActor _array_element "/Game/Athena/Items/Traps/Campfire/Blueprint/GE_Athena_Campfire_Heal.GE_Athena_Campfire_Heal_C") ; Heal
            )))
        ;[1 1 1] Overlap Pawns Loop Complete. Increment
        (Variables|Default|SetNumTicks (+ (Variables|Default|GetNumTicks) 1))
        (else
          ;[1 1 1] Reached Max Ticks. Clear Timer.
          (Variables|Default|SetIsActive false)
          (Utilities|Time|ClearTimerbyFunctionName 0 "HealTicks"))))))

(fn InitCampfireEffects
  (bind _audio (Variables|Default|GetAudio))
  (bind _returnvalue (Transformation|GetWorldLocation _audio))
  (Networking|CanExecuteCosmeticEvents
    (:True
      (if (Variables|Default|GetIsActive)
        ;[1 1 1] Turn On and Off FX and sounds based on active state change.
        (Components|Activation|SetActive (Variables|Default|GetP_Athena_Campfire) true false)
        (Components|Activation|SetActive (Variables|Default|GetP_Smoke) true false)
        (Components|Activation|Activate (Variables|Default|GetAudio) false)
        (CallFunction|CharLogs)
        (Rendering|SetRenderCustomDepth (Variables|Default|GetP_Smoke) true)
        (else
          (Components|Activation|SetActive (Variables|Default|GetP_Athena_Campfire) false false)
          (Components|Activation|SetActive (Variables|Default|GetP_Smoke) false false)
          (Components|Activation|Activate (Variables|Default|GetP_Athena_Campfire_Doused) false)
          (Development|SetHiddeninGame (Variables|Default|GetWood) true false)
          (Audio|Components|Audio|FadeOut _audio 2 0.0 "Linear" :!class /Script/Engine.AudioComponent)
          (Audio|PlaySoundatLocation (Variables|Default|GetCampfireDousedSound) _returnvalue)
          (Audio|PlaySoundatLocation (Variables|Default|GetWoodRemovalSound) _returnvalue)
          (Effects|SpawnEmitteratLocation (Variables|Default|GetWoodRemovalFX) (Transformation|GetWorldLocation (Variables|Default|GetWood)) (Transformation|GetWorldRotation (Variables|Default|GetWood)) "0.650000,0.650000,0.650000")
          (CallFunction|DousedTime)
          (Rendering|SetRenderCustomDepth (Variables|Default|GetP_Smoke) false))))))

(fn OnRep_IsActive
  (CallFunction|InitCampfireEffects))

(fn ConstructionScript)
