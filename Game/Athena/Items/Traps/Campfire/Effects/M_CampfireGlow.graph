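# Unlit translucent particle sprite material for the glow effect of the campfire trap.
# Color multiplies particle color by morning, day, evening and night tints from Color_Time_of_Day, then passes through MF_FX_EmissiveAdjustments.
# Opacity subtracts an offset sphere mask from a centered one, both warped by upward or downward panning noise, then adds a faint undistorted halo.
# Particle alpha, camera-distance fade and soft depth fade scale that shape.
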Material /Game/Athena/Items/Traps/Campfire/Effects/M_CampfireGlow
graph {
 properties {
  bUsedWithNanite=true
  bUsedWithParticleSprites=true
  blendMode=BLEND_Translucent
  pixelDepthOffsetMode=PDOM_Legacy
  shadingModel=MSM_Unlit
  usedShadingModels=Unlit
 }
 nodes [
  n1 = VectorParameter() {parameterName=Morning Tint, defaultValue=<rgba 4 4 4 0>}
  n2 = VectorParameter() {parameterName=Day Tint, defaultValue=<rgba 5 5 5 0>}
  n3 = VectorParameter() {parameterName=EveningTint, defaultValue=<rgba 1 1 1 0>}
  n4 = VectorParameter() {parameterName=NightTint, defaultValue=<rgba 1 1 1 0>}
  n5 = ParticleColor()
  MaterialFunctionCall(Emissive=Multiply(
    A=MaterialFunctionCall(Color_Morning=@n1.RGB, Color_Day=@n2.RGB, Color_Evening=@n3.RGB, Color_Night=@n4.RGB) {materialFunction=&Color_Time_of_Day}.Result
    B=@n5.RGB
   )) {materialFunction=&MF_FX_EmissiveAdjustments}
  n6 = TextureSample(UVs=Panner() {speedY=0.6}) {texture=&T_NoisePack_01}
  n7 = Add(
   A=TextureCoordinate()
   B=ConstantBiasScale(None=MaterialFunctionCall(X=@n6.R, Y=@n6.R) {materialFunction=&MakeFloat2}.Result) {bias=-0.5, scale=0.05}
  )
  n8 = Constant2Vector() {g=0.65, r=0.5}
  n9 = Constant() {r=0.4}
  n10 = Constant()
  n11 = Subtract(
   A=SphereMask(A=@n7, B=Constant() {r=0.5}, Radius=Constant() {r=0.5}, Hardness=Constant() {r=0.2})
   B=Multiply(A=SphereMask(A=@n7, B=@n8, Radius=@n9, Hardness=@n10)) {constB=2}
  )
  n12 = SphereMask(A=TextureCoordinate(), B=Constant() {r=0.5}, Radius=Constant() {r=0.5}, Hardness=Constant())
  n13 = Multiply(
   A=Add(A=Power(Base=@n11), B=Multiply(A=@n12) {constB=0.15})
   B=SphereMask(A=TextureCoordinate(), B=Constant() {r=0.5}, Radius=Constant() {r=0.5}, Hardness=Constant() {r=0.7})
  )
  n14 = MaterialFunctionCall(Fade Length=Constant() {r=2560}, Fade Offset=Constant() {r=2500}) {materialFunction=&CameraDepthFade}
  Multiply(A=Multiply(A=@n5.A, B=Multiply(A=@n13, B=@n14.Result)), B=DepthFade(FadeDistance=Constant() {r=12}))
 ]
}
