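# Niagara system for a jungle torch environment effect, stored with the Athena campfire trap effects.
# It has two emitters, Fire and Embers, with no renderers listed in this summary.
# A user Input Color parameter, orange by default, is exposed alongside an HSV_Offset module, which suggests adjustable tinting.
# SystemState and a random float range input are the only other listed logic.
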NiagaraSystem /Game/Athena/Items/Traps/Campfire/Effects/NSLW_Env_Jungle_Torch
dependencies {
 usedDataInterfaces []
 usedDynamicInputs [&RandomRangeFloat]
 usedModules [/Game/Effects/Niagara/Modules/Water/HSV_Offset, &SystemState]
}
emitters {Embers {rendererClasses [], renderers []}, Fire {rendererClasses [], renderers []}}
summary {
 systemName=NSLW_Env_Jungle_Torch
 emitters [{emitterName=Embers, rendererClasses []}, {emitterName=Fire, rendererClasses []}]
 userVariables [User.Input Color {type=LinearColor, description="", defaultValue {struct=LinearColor, value=<rgba 1 0.2502 0 0>}}]
}
